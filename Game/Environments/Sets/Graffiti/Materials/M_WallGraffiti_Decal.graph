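# Translucent deferred decal material for wall graffiti, with a swappable Graffiti texture and an Alpha scalar.
# A tiled brick texture's red channel is raised to a power, scaled, floored and inverted, forming a hard-stepped mask that keeps only its darker areas.
# That mask multiplies the graffiti alpha and Alpha scalar, breaking up the decal with the brick pattern.
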Material /Game/Environments/Sets/Graffiti/Materials/M_WallGraffiti_Decal
graph {
 properties {blendMode=BLEND_Translucent, materialDomain=MD_DeferredDecal, pixelDepthOffsetMode=PDOM_Legacy, usedShadingModels=Default Lit}
 nodes [
  Constant()
  Constant() {r=0.5}
  n1 = TextureSample(UVs=TextureCoordinate() {uTiling=1.5}) {texture=/Game/Packages/AB_Fortnite/Textures/T_BrickWall_04_D}
  n2 = Floor(None=Multiply(A=Power(Base=@n1.R) {constExponent=1.2}) {constB=7})
  Multiply(
   A=Multiply(
    A=TextureSampleParameter2D() {parameterName=Graffiti, texture=/Game/Environments/Sets/Graffiti/Textures/T_Graffiti_Cooties}.A
    B=OneMinus(None=@n2)
   )
   B=ScalarParameter() {defaultValue=1, parameterName=Alpha}
  )
 ]
}
